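# Physics asset for the JadedHare backpack cosmetic, giving collision and joint limits to a figure with a head, two-segment ears, arms and legs.
# Small Base_ capsules at head, shoulders and hips link to Body_base with free-rotation constraints and act as roots for the Dyn_ chains.
# Dyn_ capsules chain outward with limited swing; ears and limbs lock twist, while the head allows only a narrow twist.
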PhysicsAsset /BRCosmetics/Accessories/FORT_Backpacks/Backpack_F_MED_JadedHare/Meshes/F_MED_JadedHare_Pack_Physics
bodies {
 Base_head [{shapeType=Capsule, length=0.505, radius=0.505, shapeName=Base_head_capsule, rotation=<rot 0 -90 0>}]
 Base_thigh_l [{shapeType=Capsule, length=0.505, radius=0.505, shapeName=Base_thigh_l_capsule, rotation=<rot 0 -90 0>}]
 Base_thigh_r [{shapeType=Capsule, length=0.505, radius=0.505, shapeName=Base_thigh_r_capsule, rotation=<rot 0 -90 0>}]
 Base_upperarm_l [{shapeType=Capsule, length=0.505, radius=0.505, shapeName=Base_upperarm_l_capsule, rotation=<rot 0 -90 0>}]
 Base_upperarm_r [{shapeType=Capsule, length=1.505, radius=0.505, shapeName=Base_upperarm_r_capsule, rotation=<rot 0 -90 0>}]
 Body_base [{shapeType=Capsule, length=8.0743, radius=3.6665, shapeName=Body_base_capsule, center=<3.5216 0 0>, rotation=<rot -90 -179.9802 179.9803>}]
 Dyn_calf_l [{shapeType=Capsule, length=3, radius=2, shapeName=Dyn_calf_l_capsule, center=<1.6526 0 -0.1446>, rotation=<rot -85 -179.9997 179.9999>}]
 Dyn_calf_r [{shapeType=Capsule, length=3, radius=2, shapeName=Dyn_calf_r_capsule, center=<3.0362 0 -0.2656>, rotation=<rot -85 -180 -179.9999>}]
 Dyn_ear_l_01 [{
   shapeType=Capsule
   length=2.5694
   radius=2.2062
   shapeName=Dyn_ear_l_01_capsule
   center=<-2.1853 0.3496 0.0572>
   rotation=<rot 90 69.4434 69.4429>
  }]
 Dyn_ear_l_02 [{
   shapeType=Capsule
   length=3.4574
   radius=2.3939
   shapeName=Dyn_ear_l_02_capsule
   center=<-2.8853 0.5218 0.18>
   rotation=<rot 90 69.4434 69.4429>
  }]
 Dyn_ear_r_01 [{
   shapeType=Capsule
   length=2.5421
   radius=2.3022
   shapeName=Dyn_ear_r_01_capsule
   center=<-2.2203 -0.3353 0.2568>
   rotation=<rot 70 0.0005 0.0004>
  }]
 Dyn_ear_r_02 [{
   shapeType=Capsule
   length=2.2508
   radius=2.2164
   shapeName=Dyn_ear_r_02_capsule
   center=<-3.8148 -0.6249 1.1418>
   rotation=<rot 70 0.0008 0.0006>
  }]
 Dyn_head [{shapeType=Capsule, length=0.7054, radius=8.3843, shapeName=Dyn_head_capsule, center=<-6.1328 -0.8341 0>}]
 Dyn_lowerarm_l [{
   shapeType=Capsule
   length=4.4461
   radius=1.581
   shapeName=Dyn_lowerarm_l_capsule
   center=<2.8656 -0.8987 0.8419>
   rotation=<rot 0 67.0094 -66.1694>
  }]
 Dyn_lowerarm_r [{
   shapeType=Capsule
   length=4.0844
   radius=1.5814
   shapeName=Dyn_lowerarm_r_capsule
   center=<2.5462 -0.7873 -0.2332>
   rotation=<rot 0 -110.9 -82.1164>
  }]
 Dyn_thigh_l [{shapeType=Capsule, length=3, radius=2, shapeName=Dyn_thigh_l_capsule, center=<2.3171 0 -0.2027>, rotation=<rot -85 -179.9997 179.9999>}]
 Dyn_thigh_r [{shapeType=Capsule, length=3, radius=2, shapeName=Dyn_thigh_r_capsule, center=<2.8084 0 -0.2457>, rotation=<rot -85 -180 -179.9999>}]
 Dyn_upperarm_l [{
   shapeType=Capsule
   length=4.462
   radius=1.0572
   shapeName=Dyn_upperarm_l_capsule
   center=<3.4479 0.0067 -0.8829>
   rotation=<rot 0.0001 -90.6775 -104.9542>
  }]
 Dyn_upperarm_r [{
   shapeType=Capsule
   length=4.6102
   radius=0.9619
   shapeName=Dyn_upperarm_r_capsule
   center=<2.3897 0.1389 0.8398>
   rotation=<rot 0 -84.3915 -74.7477>
  }]
}
constraints [
 {
  bone1Name=Base_head
  bone2Name=Body_base
  swing1LimitDegrees=45
  swing1Motion=Free
  swing2LimitDegrees=45
  swing2Motion=Free
  twistLimitDegrees=45
  twistMotion=Free
 }
 {bone1Name=Dyn_head, bone2Name=Base_head, swing1LimitDegrees=15, swing2LimitDegrees=5, twistLimitDegrees=5}
 {bone1Name=Dyn_ear_l_01, bone2Name=Dyn_head, swing1LimitDegrees=15, swing2LimitDegrees=10, twistLimitDegrees=45, twistMotion=Locked}
 {bone1Name=Dyn_ear_l_02, bone2Name=Dyn_ear_l_01, swing1LimitDegrees=15, swing2LimitDegrees=10, twistLimitDegrees=45, twistMotion=Locked}
 {bone1Name=Dyn_ear_r_01, bone2Name=Dyn_head, swing1LimitDegrees=15, swing2LimitDegrees=10, twistLimitDegrees=45, twistMotion=Locked}
 {bone1Name=Dyn_ear_r_02, bone2Name=Dyn_ear_r_01, swing1LimitDegrees=20, swing2LimitDegrees=10, twistLimitDegrees=45, twistMotion=Locked}
 {
  bone1Name=Base_thigh_l
  bone2Name=Body_base
  swing1LimitDegrees=45
  swing1Motion=Free
  swing2LimitDegrees=10
  swing2Motion=Free
  twistLimitDegrees=45
  twistMotion=Free
 }
 {bone1Name=Dyn_thigh_l, bone2Name=Base_thigh_l, swing1LimitDegrees=35, swing2LimitDegrees=20, twistLimitDegrees=45, twistMotion=Locked}
 {bone1Name=Dyn_calf_l, bone2Name=Dyn_thigh_l, swing1LimitDegrees=35, swing2LimitDegrees=20, twistLimitDegrees=45, twistMotion=Locked}
 {
  bone1Name=Base_thigh_r
  bone2Name=Body_base
  swing1LimitDegrees=45
  swing1Motion=Free
  swing2LimitDegrees=10
  swing2Motion=Free
  twistLimitDegrees=45
  twistMotion=Free
 }
 {bone1Name=Dyn_thigh_r, bone2Name=Base_thigh_r, swing1LimitDegrees=35, swing2LimitDegrees=20, twistLimitDegrees=45, twistMotion=Locked}
 {bone1Name=Dyn_calf_r, bone2Name=Dyn_thigh_r, swing1LimitDegrees=35, swing2LimitDegrees=20, twistLimitDegrees=45, twistMotion=Locked}
 {
  bone1Name=Base_upperarm_l
  bone2Name=Body_base
  swing1LimitDegrees=15
  swing1Motion=Free
  swing2LimitDegrees=15
  swing2Motion=Free
  twistLimitDegrees=45
  twistMotion=Free
 }
 {bone1Name=Dyn_upperarm_l, bone2Name=Base_upperarm_l, swing1LimitDegrees=15, swing2LimitDegrees=25, twistLimitDegrees=45, twistMotion=Locked}
 {bone1Name=Dyn_lowerarm_l, bone2Name=Dyn_upperarm_l, swing1LimitDegrees=15, swing2LimitDegrees=25, twistLimitDegrees=45, twistMotion=Locked}
 {
  bone1Name=Base_upperarm_r
  bone2Name=Body_base
  swing1LimitDegrees=45
  swing1Motion=Free
  swing2LimitDegrees=45
  swing2Motion=Free
  twistLimitDegrees=45
  twistMotion=Free
 }
 {bone1Name=Dyn_upperarm_r, bone2Name=Base_upperarm_r, swing1LimitDegrees=15, swing2LimitDegrees=25, twistLimitDegrees=45, twistMotion=Locked}
 {bone1Name=Dyn_lowerarm_r, bone2Name=Dyn_upperarm_r, swing1LimitDegrees=15, swing2LimitDegrees=25, twistLimitDegrees=45, twistMotion=Locked}
]
