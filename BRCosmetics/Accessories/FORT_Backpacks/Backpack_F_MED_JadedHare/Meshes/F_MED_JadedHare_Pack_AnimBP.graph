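# Animation Blueprint for the JadedHare backpack cosmetic, adding secondary physics motion to its head, ears, arms and legs.
# The input pose gets additive root offsets, is cached, then feeds either a RigidBody simulation or a chain of per-bone AnimDynamics nodes.
# AnimDynamics is selected in the front end or when rigid bodies are disabled.
# The event graph also updates gravity override, spine-derived BackpackRotate and simulation space.
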AnimBlueprint /BRCosmetics/Accessories/FORT_Backpacks/Backpack_F_MED_JadedHare/Meshes/F_MED_JadedHare_Pack_AnimBP
target_skeleton=&/F_MED_JadedHare_Pack_Skeleton
graphs [
 AnimGraph {nodes [
   n1 = RigidBody!(ComponentPose=Local To Component(LocalPose=Use cached pose 'Common Cache'()), Alpha, OverrideWorldGravity:Vector="0,0,0", ExternalForce) -> (Pose) {
    bClampLinearTranslationLimitToRefPose=true
    bOverrideWorldGravity=true
    overridePhysicsAsset=&/F_MED_JadedHare_Pack_Physics
    simulationSpace=BaseBoneSpace
    alphaBoolBlend {blendOption=Linear}
    baseBoneRef {boneName=pack}
    simSpaceSettings {maxAngularAcceleration=40, maxAngularVelocity=10, maxLinearAcceleration=1300, maxLinearVelocity=750, worldAlpha=1}
   }
   n2 = Anim Dynamics - Bone: Dyn_head!(
    SimulationSpace:AnimPhysSimSpaceType Enum=BoneRelative
    ComponentPose=Local To Component(LocalPose=Use cached pose 'Common Cache'())
    Alpha=1
    ExternalForce="0,0,0"
   ) -> (Pose) {
    angularDampingOverride=0.99
    angularSpringConstant=300
    bAngularSpring=true
    bOverrideAngularDamping=true
    bUseGravityOverride=true
    simulationSpace=BoneRelative
    alphaBoolBlend {blendOption=Linear}
    boundBone {boneName=Dyn_head}
    componentAppliedLinearAccClamp=<1300 1300 1300>
    componentLinearAccScale=<0.1 0.1 0.4>
    componentLinearVelScale=<0.1 0.1 0.4>
    gravityOverride {z=-980}
    physicsBodyDefinitions [{
      boundBone {boneName=Dyn_head}
      boxExtents=<10 10 10>
      constraintSetup {angularLimitsMax=<5 0 20>, angularLimitsMin=<-5 0 -20>, angularTarget=<1 0 0>}
      localJointOffset=<1.4 0 0>
     }]
   }
   n3 = Anim Dynamics - Bone: Dyn_ear_l_01!(SimulationSpace:AnimPhysSimSpaceType Enum=BoneRelative, ComponentPose=@n2, Alpha=1, ExternalForce="0,0,0") -> (Pose) {
    angularDampingOverride=0.99
    angularSpringConstant=300
    bAngularSpring=true
    bOverrideAngularDamping=true
    bUseGravityOverride=true
    simulationSpace=BoneRelative
    alphaBoolBlend {blendOption=Linear}
    boundBone {boneName=Dyn_ear_l_01}
    componentAppliedLinearAccClamp=<1300 1300 1300>
    componentLinearAccScale=<0.1 0.1 0.4>
    componentLinearVelScale=<0.1 0.1 0.4>
    gravityOverride {z=-980}
    physicsBodyDefinitions [{
      boundBone {boneName=Dyn_ear_l_01}
      boxExtents=<9 9 9>
      constraintSetup {angularLimitsMax=<0 0 20>, angularLimitsMin=<0 0 -20>, angularTarget=<1 0 0>}
      localJointOffset=<-2 -0.3 0>
     }]
   }
   n4 = Anim Dynamics - Bone: Dyn_ear_l_02!(SimulationSpace:AnimPhysSimSpaceType Enum=BoneRelative, ComponentPose=@n3, Alpha=1, ExternalForce="0,0,0") -> (Pose) {
    angularDampingOverride=0.99
    angularSpringConstant=300
    bAngularSpring=true
    bOverrideAngularDamping=true
    bUseGravityOverride=true
    simulationSpace=BoneRelative
    alphaBoolBlend {blendOption=Linear}
    boundBone {boneName=Dyn_ear_l_02}
    componentAppliedLinearAccClamp=<1300 1300 1300>
    componentLinearAccScale=<0.1 0.1 0.4>
    componentLinearVelScale=<0.1 0.1 0.4>
    gravityOverride {z=-980}
    physicsBodyDefinitions [{
      boundBone {boneName=Dyn_ear_l_02}
      boxExtents=<9 9 9>
      constraintSetup {angularLimitsMax=<0 1 11>, angularLimitsMin=<0 -1 -11>, angularTarget=<1 0 0>}
      localJointOffset=<-2 -0.3 0>
     }]
   }
   n5 = Anim Dynamics - Bone: Dyn_ear_r_01!(SimulationSpace:AnimPhysSimSpaceType Enum=BoneRelative, ComponentPose=@n4, Alpha=1, ExternalForce="0,0,0") -> (Pose) {
    angularDampingOverride=0.99
    angularSpringConstant=300
    bAngularSpring=true
    bOverrideAngularDamping=true
    bUseGravityOverride=true
    simulationSpace=BoneRelative
    alphaBoolBlend {blendOption=Linear}
    boundBone {boneName=Dyn_ear_r_01}
    componentAppliedLinearAccClamp=<1300 1300 1300>
    componentLinearAccScale=<0.1 0.1 0.4>
    componentLinearVelScale=<0.1 0.1 0.4>
    gravityOverride {z=-980}
    physicsBodyDefinitions [{
      boundBone {boneName=Dyn_ear_r_01}
      boxExtents=<9 9 9>
      constraintSetup {angularLimitsMax=<0 0 20>, angularLimitsMin=<0 0 -20>, angularTarget=<1 0 0>}
      localJointOffset=<-2 -0.3 0>
     }]
   }
   n6 = Anim Dynamics - Bone: Dyn_ear_r_02!(SimulationSpace:AnimPhysSimSpaceType Enum=BoneRelative, ComponentPose=@n5, Alpha=1, ExternalForce="0,0,0") -> (Pose) {
    angularDampingOverride=0.99
    angularSpringConstant=300
    bAngularSpring=true
    bOverrideAngularDamping=true
    bUseGravityOverride=true
    simulationSpace=BoneRelative
    alphaBoolBlend {blendOption=Linear}
    boundBone {boneName=Dyn_ear_r_02}
    componentAppliedLinearAccClamp=<1300 1300 1300>
    componentLinearAccScale=<0.1 0.1 0.4>
    componentLinearVelScale=<0.1 0.1 0.4>
    gravityOverride {z=-980}
    physicsBodyDefinitions [{
      boundBone {boneName=Dyn_ear_r_02}
      boxExtents=<9 9 9>
      constraintSetup {angularLimitsMax=<0 0 11>, angularLimitsMin=<0 0 -11>, angularTarget=<1 0 0>}
      localJointOffset=<-2 -0.3 0>
     }]
   }
   n7 = Anim Dynamics - Bone: Dyn_upperarm_l!(SimulationSpace:AnimPhysSimSpaceType Enum, ComponentPose=@n6, Alpha=1, ExternalForce="0,0,0") -> (Pose) {
    angularDampingOverride=0.99
    angularSpringConstant=300
    bAngularSpring=true
    bOverrideAngularDamping=true
    bUseGravityOverride=true
    simulationSpace=RootRelative
    alphaBoolBlend {blendOption=Linear}
    boundBone {boneName=Dyn_upperarm_l}
    componentAppliedLinearAccClamp=<3000 3000 3000>
    componentLinearAccScale=<0.4 0.4 0.4>
    componentLinearVelScale=<0.4 0.4 0.4>
    gravityOverride {z=-980}
    physicsBodyDefinitions [{
      boundBone {boneName=Dyn_upperarm_l}
      boxExtents=<10 10 10>
      constraintSetup {coneAngle=1, angularLimitsMax=<5 35 3>, angularLimitsMin=<-5 4 -30>, angularTarget=<1 0 0>}
      localJointOffset=<1 0 0>
     }]
   }
   n8 = Anim Dynamics - Bone: Dyn_lowerarm_l!(SimulationSpace:AnimPhysSimSpaceType Enum, ComponentPose=@n7, Alpha=1, ExternalForce="0,0,0") -> (Pose) {
    angularDampingOverride=0.99
    angularSpringConstant=300
    bAngularSpring=true
    bOverrideAngularDamping=true
    bUseGravityOverride=true
    simulationSpace=RootRelative
    alphaBoolBlend {blendOption=Linear}
    boundBone {boneName=Dyn_lowerarm_l}
    componentAppliedLinearAccClamp=<3000 3000 3000>
    componentLinearAccScale=<0.4 0.4 0.4>
    componentLinearVelScale=<0.4 0.4 0.4>
    gravityOverride {z=-980}
    physicsBodyDefinitions [{
      boundBone {boneName=Dyn_lowerarm_l}
      boxExtents=<10 10 10>
      constraintSetup {coneAngle=1, angularLimitsMax=<5 45 3>, angularLimitsMin=<-5 4 -30>, angularTarget=<1 0 0>}
      localJointOffset=<1 0 0>
     }]
   }
   n9 = Anim Dynamics - Bone: Dyn_upperarm_r!(SimulationSpace:AnimPhysSimSpaceType Enum, ComponentPose=@n8, Alpha=1, ExternalForce="0,0,0") -> (Pose) {
    angularDampingOverride=0.99
    angularSpringConstant=300
    bAngularSpring=true
    bOverrideAngularDamping=true
    bUseGravityOverride=true
    simulationSpace=RootRelative
    alphaBoolBlend {blendOption=Linear}
    boundBone {boneName=Dyn_upperarm_r}
    componentAppliedLinearAccClamp=<3000 3000 3000>
    componentLinearAccScale=<0.4 0.4 0.4>
    componentLinearVelScale=<0.4 0.4 0.4>
    gravityOverride {z=-980}
    physicsBodyDefinitions [{
      boundBone {boneName=Dyn_upperarm_r}
      boxExtents=<10 10 10>
      constraintSetup {coneAngle=1, angularLimitsMax=<5 -4 3>, angularLimitsMin=<-5 -35 -30>, angularTarget=<1 0 0>}
      localJointOffset=<1 0 0>
     }]
   }
   n10 = Anim Dynamics - Bone: Dyn_lowerarm_r!(SimulationSpace:AnimPhysSimSpaceType Enum, ComponentPose=@n9, Alpha=1, ExternalForce="0,0,0") -> (Pose) {
    angularDampingOverride=0.99
    angularSpringConstant=300
    bAngularSpring=true
    bOverrideAngularDamping=true
    bUseGravityOverride=true
    simulationSpace=RootRelative
    alphaBoolBlend {blendOption=Linear}
    boundBone {boneName=Dyn_lowerarm_r}
    componentAppliedLinearAccClamp=<3000 3000 3000>
    componentLinearAccScale=<0.4 0.4 0.4>
    componentLinearVelScale=<0.4 0.4 0.4>
    gravityOverride {z=-980}
    physicsBodyDefinitions [{
      boundBone {boneName=Dyn_lowerarm_r}
      boxExtents=<10 10 10>
      constraintSetup {coneAngle=1, angularLimitsMax=<5 -4 3>, angularLimitsMin=<-5 -45 -30>, angularTarget=<1 0 0>}
      localJointOffset=<0.7 0 0>
     }]
   }
   n11 = Anim Dynamics - Bone: Dyn_thigh_l!(SimulationSpace:AnimPhysSimSpaceType Enum, ComponentPose=@n10, Alpha=1, ExternalForce="0,0,0") -> (Pose) {
    angularDampingOverride=0.99
    angularSpringConstant=300
    bAngularSpring=true
    bOverrideAngularDamping=true
    bUseGravityOverride=true
    simulationSpace=RootRelative
    alphaBoolBlend {blendOption=Linear}
    boundBone {boneName=Dyn_thigh_l}
    componentAppliedLinearAccClamp=<5500 5500 5500>
    componentLinearAccScale=<0.4 0.4 0.4>
    componentLinearVelScale=<0.4 0.4 0.4>
    gravityOverride {z=-980}
    physicsBodyDefinitions [{
      boundBone {boneName=Dyn_thigh_l}
      boxExtents=<11 11 11>
      constraintSetup {coneAngle=1, angularLimitsMax=<0 20 20>, angularLimitsMin=<0 -20 -20>, angularTarget=<1 0 0>}
      localJointOffset=<1 0 0>
     }]
   }
   n12 = Anim Dynamics - Bone: Dyn_calf_l!(SimulationSpace:AnimPhysSimSpaceType Enum, ComponentPose=@n11, Alpha=1, ExternalForce="0,0,0") -> (Pose) {
    angularDampingOverride=0.99
    angularSpringConstant=300
    bAngularSpring=true
    bOverrideAngularDamping=true
    bUseGravityOverride=true
    simulationSpace=RootRelative
    alphaBoolBlend {blendOption=Linear}
    boundBone {boneName=Dyn_calf_l}
    componentAppliedLinearAccClamp=<5500 5500 5500>
    componentLinearAccScale=<0.4 0.4 0.4>
    componentLinearVelScale=<0.4 0.4 0.4>
    gravityOverride {z=-980}
    physicsBodyDefinitions [{
      boundBone {boneName=Dyn_calf_l}
      boxExtents=<11 11 11>
      constraintSetup {coneAngle=1, angularLimitsMax=<0 10 15>, angularLimitsMin=<0 -10 -15>, angularTarget=<1 0 0>}
      localJointOffset=<1 0 0>
     }]
   }
   n13 = Anim Dynamics - Bone: Dyn_thigh_r!(SimulationSpace:AnimPhysSimSpaceType Enum, ComponentPose=@n12, Alpha=1, ExternalForce="0,0,0") -> (Pose) {
    angularDampingOverride=0.99
    angularSpringConstant=300
    bAngularSpring=true
    bOverrideAngularDamping=true
    bUseGravityOverride=true
    simulationSpace=RootRelative
    alphaBoolBlend {blendOption=Linear}
    boundBone {boneName=Dyn_thigh_r}
    componentAppliedLinearAccClamp=<5500 5500 5500>
    componentLinearAccScale=<0.4 0.4 0.4>
    componentLinearVelScale=<0.4 0.4 0.4>
    gravityOverride {z=-980}
    physicsBodyDefinitions [{
      boundBone {boneName=Dyn_thigh_r}
      boxExtents=<11 11 11>
      constraintSetup {coneAngle=1, angularLimitsMax=<0 20 20>, angularLimitsMin=<0 -20 -20>, angularTarget=<1 0 0>}
      localJointOffset=<1 0 0>
     }]
   }
   n14 = Anim Dynamics - Bone: Dyn_calf_r!(SimulationSpace:AnimPhysSimSpaceType Enum, ComponentPose=@n13, Alpha=1, ExternalForce="0,0,0") -> (Pose) {
    angularDampingOverride=0.99
    angularSpringConstant=300
    bAngularSpring=true
    bOverrideAngularDamping=true
    bUseGravityOverride=true
    simulationSpace=RootRelative
    alphaBoolBlend {blendOption=Linear}
    boundBone {boneName=Dyn_calf_r}
    componentAppliedLinearAccClamp=<5500 5500 5500>
    componentLinearAccScale=<0.4 0.4 0.4>
    componentLinearVelScale=<0.4 0.4 0.4>
    gravityOverride {z=-980}
    physicsBodyDefinitions [{
      boundBone {boneName=Dyn_calf_r}
      boxExtents=<11 11 11>
      constraintSetup {coneAngle=1, angularLimitsMax=<0 10 15>, angularLimitsMin=<0 -10 -15>, angularTarget=<1 0 0>}
      localJointOffset=<1 0 0>
     }]
   }
   Output Pose(Result=Blend Poses by bool(
     BlendPose_0=Component To Local(ComponentPose=@n14)
     BlendPose_1=Component To Local(ComponentPose=@n1)
     bActiveValue=Get FR or RBNoff() -> (FR or RBNoff:bool)
    ) {blendType=Linear}) {layerGroup=None}
   n15 = "Transform (Modify) Bone - Bone: root"(ComponentPose=Local To Component(LocalPose=Input Pose()), Translation="0,1,9", Rotation="0,0,0") {
    rotationMode=BMM_Additive
    scaleMode=BMM_Additive
    translationMode=BMM_Additive
    alphaBoolBlend {blendOption=Linear}
    boneToModify {boneName=root}
    translation {y=1, z=9}
   }
   n16 = "Transform (Modify) Bone - Bone: root"(ComponentPose=@n15, Translation="0,-13,2", Rotation="0,0,35") {
    rotationMode=BMM_Additive
    scaleMode=BMM_Additive
    translationMode=BMM_Additive
    alphaBoolBlend {blendOption=Linear}
    boneToModify {boneName=root}
    rotation {roll=35}
    translation {y=-13, z=2}
   }
   Save cached pose 'Common Cache'(Pose=Component To Local(ComponentPose=@n16)) {cacheName=Common Cache}
  ]}
 EventGraph {nodes [
   Event BlueprintUpdateAnimation()
   > n1 = Sequence!(execute) -> (then_0, then_1, then_2:Exec, then_3:Exec)
   > n2 = Gravity Override State Function DEPRECATED(CharacterPartAnimInstance=Self-Reference(), GameTimeInSeconds=GetGameTimeInSeconds(), GravityOverrideParams=Get Gravity Override Params())
   > Set GravityOverride(GravityOverride=@n2)
   n3 = BreakTransform(InTransform=GetDeltaTransformFromRefPose(self=GetOwnerSkeletalMeshComponent(), BoneName=spine_05, BaseName=spine_01))
   Set BackpackRotate(execute=@n1.then_1, BackpackRotate:double=MapRangeClamped(Value=BreakRotator(InRot=@n3.Rotation).Yaw, InRangeA=-25, InRangeB=-65)) -> (Output_Get:double)
   n4 = Get bRigidBodyEnabled()
   n5 = Branch(execute=@n1.then_2, Condition=OR Boolean(A=NOT Boolean(A=@n4), B=Get bIsInFrontEnd()))
   > Set FR or RBNoff(FR or RBNoff:bool=true) -> (Output_Get:bool)
   Set FR or RBNoff(execute=@n5.else, FR or RBNoff=false)
   n6 = Branch(execute=@n1.then_3, Condition=Get bIsInFrontEnd())
   > Set SimlationSpace(SimlationSpace:AnimPhysSimSpaceType Enum=World) -> (Output_Get:AnimPhysSimSpaceType Enum)
   Set SimlationSpace(execute=@n6.else, SimlationSpace=Component)
   TryGetPawnOwner()
   Reroute Node(InputPin:Rotator=@n3.Rotation) -> (OutputPin:Rotator)
  ]}
]
variables [
 BackpackRotate {type=number}
 FR or RBNoff {type=boolean}
 Gravity Override Params {type=GravityOverrideParamsStruct}
 GravityOverride {type=Vector}
 SimlationSpace {type=AnimPhysSimSpaceType}
]
